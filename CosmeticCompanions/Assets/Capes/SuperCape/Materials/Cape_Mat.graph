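# Surface material for the SuperCape cosmetic cape in CosmeticCompanions, using Default Lit shading.
# Its graph holds a single constant red-orange colour node, with no textures or parameters.
# Usage flags enable it for skeletal meshes, clothing simulation and mesh deformers.
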Material /CosmeticCompanions/Assets/Capes/SuperCape/Materials/Cape_Mat
graph {
 properties {
  assetImportData="<self>:InterchangeAssetImportData_0"
  bUsedWithClothing=true
  bUsedWithMeshDeformer=true
  bUsedWithSkeletalMesh=true
  usedShadingModels=Default Lit
 }
 nodes [Constant3Vector() {constant=<rgba 0.6348 0.0239 0.0001 1>}]
}
