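# Physics asset for the SuperCape cosmetic, giving the cape rigid-body dynamics through left, middle and right capsule chains.
# Each chain starts at a small MSTR anchor body and links segment to segment with tight swing and twist limits.
# Angularly free cross constraints tie the lower left and right segments to the middle chain, so the strands hold together like a sheet.
# Hip, spine and back-thigh capsules supply body shapes for the cape to collide with.
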PhysicsAsset /CosmeticCompanions/Assets/Capes/SuperCape/Meshes/SuperCape_Dyn
bodies {
 dyn_cape_l_01 [{shapeType=Capsule, length=12.1067, radius=4.4117, shapeName=dyn_cape_l_01_capsule, center=<8.4289 0 4.0718>, rotation=<rot 90 0 0>}]
 dyn_cape_l_01_MSTR [{shapeType=Capsule, length=0.505, radius=0.505, shapeName=dyn_cape_l_01_MSTR_capsule, rotation=<rot 0 -90 0>}]
 dyn_cape_l_02 [{shapeType=Capsule, length=12.1067, radius=4.4117, shapeName=dyn_cape_l_01_capsule, center=<8.4289 0 4.0718>, rotation=<rot 90 0 0>}]
 dyn_cape_l_03 [{shapeType=Capsule, length=12.1067, radius=4.4117, shapeName=dyn_cape_l_01_capsule, center=<8.4289 0 4.0718>, rotation=<rot 90 0 0>}]
 dyn_cape_l_04 [{shapeType=Capsule, length=12.1067, radius=4.4117, shapeName=dyn_cape_l_01_capsule, center=<8.4289 0 4.0718>, rotation=<rot 90 0 0>}]
 dyn_cape_l_05 [{shapeType=Capsule, length=12.1067, radius=4.4117, shapeName=dyn_cape_l_01_capsule, center=<8.4289 0 4.0718>, rotation=<rot 90 0 0>}]
 dyn_cape_mid_01 [{shapeType=Capsule, length=12.1067, radius=4.3854, shapeName=dyn_cape_l_01_capsule, center=<8.4289 0 4.0718>, rotation=<rot 90 0 0>}]
 dyn_cape_mid_01_MSTR [{shapeType=Capsule, length=0.505, radius=0.505, shapeName=dyn_cape_mid_01_MSTR_capsule, rotation=<rot 0 -90 0>}]
 dyn_cape_mid_02 [{shapeType=Capsule, length=12.1067, radius=4.3854, shapeName=dyn_cape_l_01_capsule, center=<8.4289 0 4.0718>, rotation=<rot 90 0 0>}]
 dyn_cape_mid_03 [{shapeType=Capsule, length=12.1067, radius=4.3854, shapeName=dyn_cape_l_01_capsule, center=<8.4289 0 4.0718>, rotation=<rot 90 0 0>}]
 dyn_cape_mid_04 [{shapeType=Capsule, length=12.1067, radius=4.3854, shapeName=dyn_cape_l_01_capsule, center=<8.4289 0 4.0718>, rotation=<rot 90 0 0>}]
 dyn_cape_r_01 [{shapeType=Capsule, length=12.1067, radius=4.3854, shapeName=dyn_cape_l_01_capsule, center=<8.4289 0 4.0718>, rotation=<rot 90 0 0>}]
 dyn_cape_r_01_MSTR [{shapeType=Capsule, length=0.505, radius=0.505, shapeName=dyn_cape_r_01_MSTR_capsule, rotation=<rot 0 -90 0>}]
 dyn_cape_r_02 [{shapeType=Capsule, length=12.1067, radius=4.3854, shapeName=dyn_cape_l_01_capsule, center=<8.4289 0 4.0718>, rotation=<rot 90 0 0>}]
 dyn_cape_r_03 [{shapeType=Capsule, length=12.1067, radius=4.3854, shapeName=dyn_cape_l_01_capsule, center=<8.4289 0 4.0718>, rotation=<rot 90 0 0>}]
 dyn_cape_r_04 [{shapeType=Capsule, length=12.1067, radius=4.3854, shapeName=dyn_cape_l_01_capsule, center=<8.4289 0 4.0718>, rotation=<rot 90 0 0>}]
 dyn_cape_r_05 [{shapeType=Capsule, length=12.1067, radius=4.3854, shapeName=dyn_cape_l_01_capsule, center=<8.4289 0 4.0718>, rotation=<rot 90 0 0>}]
 hips [{shapeType=Capsule, length=25.7909, radius=11.2843, center=<0.2778 2.0669 0>, rotation=<rot 90 180 -172.3457>}]
 spine_02 [{shapeType=Capsule, length=18.1318, radius=12.5096, center=<3.5975 0.436 0.2938>, rotation=<rot 90 0 7.6066>}]
 spine_03 [{shapeType=Capsule, length=11.1021, radius=13.5066, center=<-1.7394 0.1373 0.2938>, rotation=<rot 90 0 14.9074>}]
 thigh_bk_l [{shapeType=Capsule, length=13.4077, radius=9.2396, center=<11.0944 -1.2018 -3.4904>, rotation=<rot 86.6152 175.2161 180>}]
 thigh_bk_r [{shapeType=Capsule, length=13.4077, radius=9.2396, center=<-11.0944 1.2018 3.4904>, rotation=<rot 86.6152 175.2161 0>}]
}
constraints [
 {bone1Name=dyn_cape_l_01, bone2Name=dyn_cape_l_01_MSTR, swing1LimitDegrees=15, swing2LimitDegrees=15, twistLimitDegrees=5}
 {bone1Name=dyn_cape_l_02, bone2Name=dyn_cape_l_01, swing1LimitDegrees=15, swing2LimitDegrees=15, twistLimitDegrees=5}
 {bone1Name=dyn_cape_l_03, bone2Name=dyn_cape_l_02, swing1LimitDegrees=15, swing2LimitDegrees=15, twistLimitDegrees=5}
 {bone1Name=dyn_cape_l_04, bone2Name=dyn_cape_l_03, swing1LimitDegrees=15, swing2LimitDegrees=15, twistLimitDegrees=5}
 {bone1Name=dyn_cape_l_05, bone2Name=dyn_cape_l_04, swing1LimitDegrees=15, swing2LimitDegrees=15, twistLimitDegrees=5}
 {bone1Name=dyn_cape_mid_01, bone2Name=dyn_cape_mid_01_MSTR, swing1LimitDegrees=15, swing2LimitDegrees=15, twistLimitDegrees=5}
 {bone1Name=dyn_cape_mid_02, bone2Name=dyn_cape_mid_01, swing1LimitDegrees=15, swing2LimitDegrees=15, twistLimitDegrees=5}
 {bone1Name=dyn_cape_mid_03, bone2Name=dyn_cape_mid_02, swing1LimitDegrees=15, swing2LimitDegrees=15, twistLimitDegrees=5}
 {bone1Name=dyn_cape_mid_04, bone2Name=dyn_cape_mid_03, swing1LimitDegrees=15, swing2LimitDegrees=15, twistLimitDegrees=5}
 {bone1Name=dyn_cape_r_01, bone2Name=dyn_cape_r_01_MSTR, swing1LimitDegrees=15, swing2LimitDegrees=15, twistLimitDegrees=5}
 {bone1Name=dyn_cape_r_02, bone2Name=dyn_cape_r_01, swing1LimitDegrees=15, swing2LimitDegrees=15, twistLimitDegrees=5}
 {bone1Name=dyn_cape_r_03, bone2Name=dyn_cape_r_02, swing1LimitDegrees=15, swing2LimitDegrees=15, twistLimitDegrees=5}
 {bone1Name=dyn_cape_r_04, bone2Name=dyn_cape_r_03, swing1LimitDegrees=15, swing2LimitDegrees=15, twistLimitDegrees=5}
 {bone1Name=dyn_cape_r_05, bone2Name=dyn_cape_r_04, swing1LimitDegrees=15, swing2LimitDegrees=15, twistLimitDegrees=5}
 {
  bone1Name=dyn_cape_l_03
  bone2Name=dyn_cape_mid_02
  swing1LimitDegrees=45
  swing1Motion=Free
  swing2LimitDegrees=45
  swing2Motion=Free
  twistLimitDegrees=45
  twistMotion=Free
 }
 {
  bone1Name=dyn_cape_r_03
  bone2Name=dyn_cape_mid_02
  swing1LimitDegrees=45
  swing1Motion=Free
  swing2LimitDegrees=45
  swing2Motion=Free
  twistLimitDegrees=45
  twistMotion=Free
 }
 {
  bone1Name=dyn_cape_l_04
  bone2Name=dyn_cape_mid_03
  swing1LimitDegrees=45
  swing1Motion=Free
  swing2LimitDegrees=45
  swing2Motion=Free
  twistLimitDegrees=45
  twistMotion=Free
 }
 {
  bone1Name=dyn_cape_r_04
  bone2Name=dyn_cape_mid_03
  swing1LimitDegrees=45
  swing1Motion=Free
  swing2LimitDegrees=45
  swing2Motion=Free
  twistLimitDegrees=45
  twistMotion=Free
 }
 {
  bone1Name=dyn_cape_r_05
  bone2Name=dyn_cape_mid_04
  swing1LimitDegrees=45
  swing1Motion=Free
  swing2LimitDegrees=45
  swing2Motion=Free
  twistLimitDegrees=45
  twistMotion=Free
 }
 {
  bone1Name=dyn_cape_l_05
  bone2Name=dyn_cape_mid_04
  swing1LimitDegrees=45
  swing1Motion=Free
  swing2LimitDegrees=45
  swing2Motion=Free
  twistLimitDegrees=45
  twistMotion=Free
 }
]
